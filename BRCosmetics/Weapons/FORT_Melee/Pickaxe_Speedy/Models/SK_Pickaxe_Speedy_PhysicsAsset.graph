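# Physics asset for the Speedy pickaxe skeletal mesh, a two-handed melee harvesting tool cosmetic.
# A single box body on the Melee_TwoHanded_Handle bone encloses the weapon as one rigid shape.
# It has no constraints, so it provides simple collision and bounds rather than any jointed or secondary motion.
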PhysicsAsset /BRCosmetics/Weapons/FORT_Melee/Pickaxe_Speedy/Models/SK_Pickaxe_Speedy_PhysicsAsset
bodies {Melee_TwoHanded_Handle [{shapeType=Box, extentX=87.4625, extentY=11.3812, extentZ=155.6932, center=<-0.0069 -2.1831 21.6063>, rotation=<rot -0.5773 -90 -0.0012>}]}
constraints []
